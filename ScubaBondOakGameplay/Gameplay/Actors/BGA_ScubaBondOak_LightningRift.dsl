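; Lightning rift actor from the ScubaBondOak guitar item that grants a one-time ability activation to player pawns overlapping its sphere.
; It pre-registers the instigator, whose use ability handles them separately, and activates a non-leader ability for each new pawn.
; Repeat pawns get a failed-action message, gated by a short delay to hide it from the original user.
; It expires by scalable lifespan, fading Niagara and audio beforehand, or upon reaching max uses.
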(event EventBeginPlay
  (bind _hf_lifespan (Variables|Default|GetHF_Lifespan))
  (bind _returnvalue (ScalableFloat|GetValueatLevel _hf_lifespan))
  ;[0 1 0.4077] Destroy after lifespan
  (Variables|Default|SetTotalUses 0)
  (Actor|SetLifeSpan _returnvalue)
  ;[1 0.4946 0] If we have a pawn instigator,  we shouldn't use collision to trigger this ability. We'll handle that in the SCubaBondOak_Use Abilty
  (Utilities|FlowControl|Sequence
    (:then_0
      (bind _asfort_pawn (Utilities|Casting|CastToFortPawn (Variables|Actor|GetInstigator)))
      (Utilities|Array|Add (Variables|Default|GetInteractingPawns) _asfort_pawn)
      ;[1 1 1] Allow unable to perform action message only after a short delay so the original RTL user doesn't see it
      (Utilities|FlowControl|Delay 0.5)
      (CallFunction|AllowActionMessage))
    (:then_1
      (Utilities|FlowControl|Delay (- _returnvalue 0.6))
      (Utilities|FlowControl|Sequence
        (:then_0
          (bind (_direction _fade) (Timeline|FXFade PlayFromStart)
            (:Update
              (Niagara|SetNiagaraVariable(Float) (Variables|Default|GetPortalFX) "EnableValue" _fade))))
        (:then_1
          (bind _audio (Variables|Default|GetAudio)
            (:then
              (Audio|Components|Audio|FadeOut _audio 0.6 0.0 "Linear" :!class /Script/Engine.AudioComponent))
            (:else)))))))

(event OnComponentBeginOverlap(Sphere)
  (params (OverlappedComponent :type Primitive%20Component%20Object%20Reference) (OtherActor :type Actor%20Object%20Reference) (OtherComp :type Primitive%20Component%20Object%20Reference) (OtherBodyIndex :type Integer) (bFromSweep :type Boolean) (SweepResult :type Hit%20Result%20Structure%20%28by%20ref%29))
  ;[0 1 0.4077] Player overlaps -> should happen instantly when the guitar user plays the item
  (if (not (Game|IsActorBeingDestroyed))
    (bind _asfort_player_pawn (Utilities|Casting|CastToFortPlayerPawn OtherActor))
    (Niagara|SpawnSystematLocation :SystemTemplate "/ScubaBondOakGameplay/Assets/VFX/NS_ScubaBondOak_RiftActivationFX.NS_ScubaBondOak_RiftActivationFX" :Location (Transformation|GetActorLocation) :Scale "1.000000,1.000000,1.000000" :bAutoDestroy true :bAutoActivate true :PoolingMethod "AutoRelease" :bPreCullCheck true)
    (bind _returnvalue (CallFunction|ProcessPawn _asfort_player_pawn))
    (if _returnvalue
      (Utilities|FlowControl|Gate true
        (:Exit
          (FailedAction|UnabletoPerformAction _asfort_player_pawn true (Variables|Default|GetFailedTag))))
      (else
        (Abilities|TryActivateAbilitybyClass (Ability|GetFortAbilitySystemComponent _asfort_player_pawn) (Variables|Default|GetActivationClassNoWeap) true) ; Activate exclusive ability for non-leader users
        (Math|Integer|IncrementInt (Variables|Default|GetTotalUses)
          (:%20%20
            (CallFunction|CheckShouldDestroy)))))))

(event Custom|AllowActionMessage
  ;[0 1 0.4077] Player overlaps -> should happen instantly when the guitar user plays the item
  (Utilities|FlowControl|Gate true
    (:Exit
      (FailedAction|UnabletoPerformAction (stale OnComponentBeginOverlap(Sphere) _asfort_player_pawn) true (Variables|Default|GetFailedTag)))))

(fn CheckShouldDestroy
  ;[0.4895 0.1317 1] Destroy if max uses is reached
  (if (>= (Variables|Default|GetTotalUses) (ScalableFloat|GetValueasInteger (Variables|Default|GetHF_MaxUses)))
    (Actor|DestroyActor)))

(fn ProcessPawn
  (params (Pawn :type Fort%20Pawn%20Object%20Reference))
  (bind _interactingpawns (Variables|Default|GetInteractingPawns))
  (bind _returnvalue (Utilities|Array|ContainsItem _interactingpawns Pawn))
  (LocalVariables|SetLocalBool _returnvalue)
  (if (not _returnvalue)
    (Utilities|Array|AddUnique _interactingpawns (Variables|GetPawn)))
  (return (LocalVariables|GetLocalBool)))

(fn ConstructionScript)
