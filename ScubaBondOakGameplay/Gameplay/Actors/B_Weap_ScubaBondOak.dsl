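; Weapon actor Blueprint for ScubaBondOak that drives the ability's cooldown state, feedback and dash/jump trail FX.
; The Event.ScubaBond.Oak.Used gameplay event starts a timer from its magnitude and binds a locally heard "can't do" trigger sound until expiry.
; ToggleCooldownState switches ambient Niagara, mesh emissive and equip-audio parameters, then broadcasts the change.
; Separate events force cooldown visuals from a GameplayCue and spawn or destroy the mesh-linked Niagara trail.
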(event EventBeginPlay
  ;[0 0.8244 1] Check when weapon goes into cooldown
  (bind (_asynctaskproxy _payload) (Ability|Async|WaitGameplayEventToActor (Game|GetInstigator) "(TagName=\"Event.ScubaBond.Oak.Used\")")
    (:EventReceived
      (CallFunction|ToggleCooldownState true)
      (Variables|Default|SetbIsCooldown? true)
      (Utilities|Time|SetTimerbyEvent (delegate Custom|CooldownExpired) (.EventMagnitude (Utilities|Struct|BreakGameplayEventData _payload)))
      ;[0.1589 0.2785 1] Used for playing the "clammed" strum if we attempt activation during cooldown.
      (EventDispatchers|BindEventtoOnPressTrigger (delegate Custom|OnPressTrigger_Event)))))

(event Custom|OnPressTrigger_Event
  (params (Weapon :type Fort%20Weapon%20Object%20Reference))
  ;[0.1589 0.2785 1] Used for playing the "clammed" strum if we attempt activation during cooldown.
  (if (and (and (and (FortWeapon|IsEquipped Weapon) (Rendering|IsVisible (Variables|Visual|GetWeaponMesh))) (stale Custom|ToggleCooldownState InCooldown?)) (Pawn|IsLocallyViewed (Game|GetInstigator)))
    (Audio|SpawnSound2D "/ScubaBondOakGameplay/Assets/Audio/SFX/MSS_ScubaBondOak_Abl_Sfx_CanNotDo.MSS_ScubaBondOak_Abl_Sfx_CanNotDo")))

(event Custom|CooldownExpired
  ;[0 0.8244 1] Check when weapon goes into cooldown
  (Variables|Default|SetbIsCooldown? false)
  (CallFunction|ToggleCooldownState false)
  ;[0.1589 0.2785 1] Used for playing the "clammed" strum if we attempt activation during cooldown.
  (EventDispatchers|UnbindEventfromOnPressTrigger (delegate Custom|OnPressTrigger_Event)))

(event Custom|ToggleCooldownState
  (params (InCooldown? :type Boolean))
  (bind _ns_scubabondoak_weaponambient (Variables|Default|GetNS_ScubaBondOak_WeaponAmbient))
  (bind _weaponmesh (Variables|Visual|GetWeaponMesh))
  ;[0 0.8244 1] Cooldown -> client only
  (Utilities|FlowControl|Sequence
    (:then_0
      (if InCooldown?
        (Variables|Default|SetbIsCooldown? true)
        (Rendering|SetVisibility _ns_scubabondoak_weaponambient false false)
        (Rendering|Material|SetVectorParameterValueonMaterials _weaponmesh "EmissiveMultiplier" "0, 0, 0")
        (Rendering|Material|SetCustomPrimitiveDataFloat _weaponmesh 0 (Utilities|Time|GetGameTimeinSeconds))
        (else
          (Variables|Default|SetbIsCooldown? false)
          (Rendering|SetVisibility _ns_scubabondoak_weaponambient true false)
          (Rendering|Material|SetVectorParameterValueonMaterials _weaponmesh "EmissiveMultiplier" "5.000000,0.000000,0.000000")))
      ;[0 0.2517 0.2082] Handle Audio
      (Class|WeaponEquipSoundComponent|ExectuteBoolParameter (Variables|Default|GetWeaponEquipSoundComponent) "InCooldown" InCooldown?)
      (Class|WeaponEquipSoundComponent|ExectuteBoolParameter (Variables|Default|GetWeaponEquipSoundComponent) "InCooldown_EquipLayer" InCooldown?))
    (:then_1
      (Default|CallOnCooldownChanged InCooldown?))
    (:then_2
      (if (and (FortWeapon|IsEquipped) (Rendering|IsVisible (Variables|Visual|GetWeaponMesh)))
        (Audio|SpawnSoundAttached (Variables|Default|GetReadyToUseAgain_Sound) (Variables|Visual|GetWeaponMesh))))
    (:then_3
      (CallFunction|OnCooldownChangedEvent InCooldown?))))

(event FortWeapon|EventOnEquippedAndAttached
  (bind _biscooldown (Variables|Default|GetbIsCooldown?))
  ;[0 0.2517 0.2082] Handle Audio
  (Utilities|FlowControl|Delay 0.2)
  (Class|WeaponEquipSoundComponent|ExectuteBoolParameter (Variables|Default|GetWeaponEquipSoundComponent) "InCooldown" _biscooldown)
  (Class|WeaponEquipSoundComponent|ExectuteBoolParameter (Variables|Default|GetWeaponEquipSoundComponent) "InCooldown_EquipLayer" _biscooldown)
  (if (not _biscooldown)
    (Audio|SpawnSoundAttached "/ScubaBondOakGameplay/Assets/Audio/SFX/MSS_ScubaBondOak_Abl_Sfx_Equip_SparksLayer.MSS_ScubaBondOak_Abl_Sfx_Equip_SparksLayer" (Variables|Visual|GetWeaponMesh))))

(event Custom|OnCooldownChanged_Event
  (params (IsCooldown? :type Boolean))
  ;[0 0.2517 0.2082] Handle Audio
  (Variables|Default|SetbIsCooldown? IsCooldown?))

(event Custom|ToggleCooldownStateFromGC
  ;[0 0.8244 1] Cooldown -> client only
  ;;[0 0.6667 1] ToggleCooldownStatFromEMG GC because we don't want this FX to disappear when the player flies in the sky. - Yi.Xie 4.18.2024
  (Utilities|FlowControl|Delay 0.1)
  (Rendering|SetVisibility (Variables|Default|GetNS_ScubaBondOak_WeaponAmbient) false false)
  (Class|WeaponEquipSoundComponent|ExectuteBoolParameter (Variables|Default|GetWeaponEquipSoundComponent) "InCooldown" true)
  (Variables|Default|SetbIsCooldown? true)
  (Class|WeaponEquipSoundComponent|ExectuteBoolParameter (Variables|Default|GetWeaponEquipSoundComponent) "InCooldown_EquipLayer" true)
  (Development|PrintString "ExecuteCooldownFunctionFromGC"))

(event Custom|ActiveTrailFX
  (bind _trailfx (Variables|Default|GetTrailFX))
  (Utilities|IsValid _trailfx
    (:Is%20Valid
      (Components|Activation|Deactivate _trailfx)
      (Components|DestroyComponent _trailfx))
    (:Is%20Not%20Valid
      _))
  (bind _returnvalue (Effects|Components|ParticleSystem|SpawnFXAtLocation :FXSystemAsset "/ScubaBondOakGameplay/Assets/VFX/NS_ScubaBondOak_TrailFX.NS_ScubaBondOak_TrailFX" :Location (Transformation|GetActorLocation) :Scale "1.000000,1.000000,1.000000" :bAutoDestroy true :PoolingMethod "AutoRelease" :bAutoActivate false))
  (Variables|Default|SetLastJumpTimeForFX (Utilities|Time|GetRealTimeSeconds))
  (Variables|Default|SetTrailFX _returnvalue)
  (bind _asniagara_particle_system_component (Utilities|Casting|CastToNiagaraComponent _returnvalue))
  (Niagara|SetNiagaraSkeletalMeshComponent _asniagara_particle_system_component "WeapMesh" (Variables|Visual|GetWeaponMesh))
  (Components|Activation|SetActive _returnvalue true true)
  (Development|PrintString "Active Trail FX" true true "(R=1.000000,G=0.209458,B=0.000000,A=1.000000)" 5e+01))

(event Custom|DeactiveTrailFX
  (bind _trailfx (Variables|Default|GetTrailFX))
  (Utilities|IsValid _trailfx
    (:Is%20Valid
      (Components|Activation|Deactivate _trailfx)
      (Components|DestroyComponent _trailfx)
      (Variables|Default|SetTrailFX 0)
      (Development|PrintString "DEActive Trail FX" true true "(R=0.000000,G=1.000000,B=0.299024,A=1.000000)" 5e+01))))

(fn GetDashesPerformed
  (return (Variables|Default|GetDashesPerformed)))

(fn GetReticleColorOverride
  (params (HitResults :type Array%20of%20Hit%20Result%20Structures))
  (bind (_b _g _r) (Utilities|Struct|BreakColor (Reticle|GetReticleColor)))
  (return (Utilities|Struct|MakeColor _b _g _r) false))

(fn ConstructionScript)
